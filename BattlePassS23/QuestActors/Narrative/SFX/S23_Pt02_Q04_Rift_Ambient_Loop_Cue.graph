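# Looping ambient sound cue for a rift in a Season 23 narrative quest (Part 2, Quest 4).
# A single wave player loops the rift ambient SFX wave as the cue's root, with no mixing or randomization.
# It plays through the shared Season 23 quest attenuation and the Fort_Quests sound class.
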SoundCue /BattlePassS23/QuestActors/Narrative/SFX/S23_Pt02_Q04_Rift_Ambient_Loop_Cue
root=WavePlayer_0
nodes [{id=WavePlayer_0, properties {bLooping=true, soundWaveAssetPtr=&/S23_Pt02_Q04_Rift_Ambient_Loop_SFX}}]
properties {
 attenuationSettings=&/S23_Quests_Atten
 maxDistance=1300
 soundClassObject=&Fort_Quests
 volumeMultiplier=1
}
