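# Sound cue for the forge alloy effect in a Season 23 narrative quest, Part 3 Quest 3.
# A single wave player feeds the Forge Alloy sound wave straight to the output, with no randomization or mixing.
# The cue uses the shared Season 23 quest attenuation and the Fort quests sound class, with a boosted volume.
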SoundCue /BattlePassS23/QuestActors/Narrative/SFX/S23_Pt03_Q03_Forge_Alloy_Cue
root=WavePlayer_0
nodes [{id=WavePlayer_0, properties {soundWaveAssetPtr=&/S23_Pt03_Q03_Forge_Alloy_SFX}}]
properties {
 attenuationSettings=&/S23_Quests_Atten
 maxDistance=1300
 soundClassObject=&Fort_Quests
 volumeMultiplier=1.5
}
