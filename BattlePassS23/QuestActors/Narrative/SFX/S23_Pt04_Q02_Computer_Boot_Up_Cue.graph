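# Sound cue for the computer boot-up effect in a Season 23 Battle Pass quest, likely part 4, quest 2.
# A single wave player feeds the output directly with the boot-up sound wave, with no randomization or modulation.
# Shared Season 23 quest attenuation limits its range, and it routes through the Fort quests sound class.
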SoundCue /BattlePassS23/QuestActors/Narrative/SFX/S23_Pt04_Q02_Computer_Boot_Up_Cue
root=WavePlayer_0
nodes [{id=WavePlayer_0, properties {soundWaveAssetPtr=&/S23_Pt04_Q02_Boot_Up}}]
properties {
 attenuationSettings=&/S23_Quests_Atten
 maxDistance=1300
 soundClassObject=&Fort_Quests
 volumeMultiplier=1
}
