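# Looping ambient SoundCue for a small rift's tonal hum, used by a Season 23 Battle Pass quest actor.
# A single looping Wave Player feeds the root directly, with no mixing or randomization.
# The cue boosts volume, raises pitch slightly, routes to the Fort_Quests sound class and uses the shared Season 23 quest attenuation asset.
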SoundCue /BattlePassS23/QuestActors/Narrative/SFX/S23_Pt04_Q02_Rift_Small_Tonal_Amb_Loop_Cue
root=WavePlayer_0
nodes [{
  id=WavePlayer_0
  properties {bLooping=true, soundWaveAssetPtr=&/S23_Pt04_Q02_Rift_Small_Tonal_Amb_Loop_SFX}
 }]
properties {
 attenuationSettings=&/S23_Quests_Atten
 maxDistance=1300
 pitchMultiplier=1.1
 soundClassObject=&Fort_Quests
 volumeMultiplier=2.5
}
